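# Conversation database for the NPC Stashd, defining the dialogue that opens when a player talks to this character.
# An entry point leads to one speech node with a default lift-offer line.
# Contextual messages replace it on a first-ever conversation, prioritised by a curve row, or when the player wears particular character cosmetics.
# The speech then links out to shared quest and basic-services conversations by entry tag.
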ConversationDatabase /NPCLibrary/NPCs/Stashd/Stashd/NPCLibrary_Conversation_Stashd
entry_points [{tag=Conversation.Entry.Stashd, destinations [#1]}]
nodes [
 ConversationEntryPointNode_7 {
  id=#1
  class=ConversationEntryPointNode
  outputs [#2]
  properties {entryTag {tagName=Conversation.Entry.Stashd}}
 }
 Speech_C_4 {
  id=#2
  class=Speech_C
  outputs [#3, #4]
  properties {
   nodeName=""
   generalConfig {
    defaultMessage="Need a lift? "
    contextualMessages [
     {
      message=I'm with the Peace Syndicate now... But I still freelance.
      contextRequirements [{requirement {class=FortControllerRequirement_IsFirstEverConversationWithNPC}}]
      priority {curve {rowName=Tandem.Speech.Priority.FirstEverInteraction}}
     }
     {
      message=I've been thinking about it. Are you sure there's no other way?
      contextRequirements [{requirement {
         class=FortControllerRequirement_HasCID
         properties {allowedCIds [&CID_A_425_Athena_Commando_F_BlueJay]}
        }}]
      priority {value=0, curve {curveTable=null, rowName=None}}
     }
     {
      message=Take your complaints up with Evie. I'm just the driver.
      contextRequirements [{requirement {class=FortControllerRequirement_HasCID, properties {allowedCIds [/BRCosmetics/Athena/Items/Cosmetics/Characters/Character_Candor]}}}]
      priority {value=0, curve {curveTable=null, rowName=None}}
     }
     {
      message=You got a driver's licence and a love of mayhem?
      contextRequirements [{requirement {class=FortControllerRequirement_HasCID, properties {allowedCIds [&Character_ChillCat]}}}]
      priority {value=0, curve {curveTable=null, rowName=None}}
     }
     {
      message=The Syndicate's recruiting. I could put a word in for you.
      contextRequirements [{requirement {class=FortControllerRequirement_HasCID, properties {allowedCIds [&Character_Bites]}}}]
      priority {value=0, curve {curveTable=null, rowName=None}}
     }
     {
      message=Let me know if you ever need a getaway driver.
      contextRequirements [{requirement {class=FortControllerRequirement_HasCID, properties {allowedCIds [&Character_PinkSpike]}}}]
      priority {value=0, curve {curveTable=null, rowName=None}}
     }
     {
      message=Pretty cool ride. How fast can it go?
      contextRequirements [{requirement {class=FortControllerRequirement_HasCID, properties {allowedCIds [&Character_BadBear]}}}]
      priority {value=0, curve {curveTable=null, rowName=None}}
     }
     {
      message=Do you really think you're quick enough?
      contextRequirements [{requirement {class=FortControllerRequirement_HasCID, properties {allowedCIds [&Character_TheHerald]}}}]
      priority {value=0, curve {curveTable=null, rowName=None}}
     }
     {
      message="Hey, we kinda match."
      contextRequirements [{requirement {class=FortControllerRequirement_HasCID, properties {allowedCIds [&Character_RoseDust]}}}]
      priority {value=0, curve {curveTable=null, rowName=None}}
     }
    ]
   }
  }
 }
 ConversationLinkNode_0 {id=#4, class=ConversationLinkNode, properties {remoteEntryTag {tagName=Conversation.Entry.BasicServices}}}
 ConversationLinkNode_1 {id=#3, class=ConversationLinkNode, properties {remoteEntryTag {tagName=Conversation.Entry.NPCLibrary.Quests}}}
]
